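# Niagara system for the PopStar vehicle boost, a star-themed exhaust effect placed at filtered skeletal-mesh sockets with a car offset.
# Two looping mesh glow emitters form pulsing heads, while others emit expanding star outlines, backward-flying star sprites and ribbon trails.
# Colors switch between default and painted looks through user parameters and painted-variant override tables.
# Ribbon outline brightness varies with the camera's angle to the owner's forward axis.
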NiagaraSystem /VehicleCosmetics/Boosts/Boost_PopStar/FX/NS_Boost_PopStar
dependencies {
 usedDataInterfaces [
  &NiagaraDataInterfaceCamera
  &NiagaraDataInterfaceCurve
  &NiagaraDataInterfaceArrayColor
  &NiagaraDataInterfaceArrayFloat3
  &NiagaraDataInterfaceSkeletalMesh
 ]
 usedDynamicInputs []
 usedModules [
  &self:SystemSpawnScript.SystemScriptSource.SystemScriptGraph.NiagaraNodeAssignment_0.SetVariables_#1
  &SystemState
 ]
}
emitters {
 StarSprites001 {
  emitterUpdateScript [
   &EmitterState()
   &VC_FilteredSocketCount()
   &SpawnRate(SpawnRate=35, Use Spawn Probability=false)
  ]
  particleSpawnScript [
   &VC_PaintedVariantsOverride() {enabled=false}
   &InitializeParticle(
    Color=&MakeCustomLinearColorFromBool(
     Bool=$User.bIsPainted
     True Color=&MultiplyLinearColor(Linear Color=$User.PaintedColor, Scale Factor=<0.4 0.4 0.4 1>)
     False Color=&SelectLinearColorFromArray(
      Array Sampling Mode=::NewEnumerator0
      Direct Array Index=?
      Color Selection Array=DataInterface() {propertyValues {ColorData [<rgba 0 1 0.1239 1>, <rgba 0.6939 0 1 1>, <rgba 1 0.5873 0.1156 1>, <rgba 0 0.4286 1 1>]}}
      Random Seed=0
     )
    )
    Lifetime=0.6
    Sprite Rotation Mode=::NewEnumerator2
    Sprite Size Mode=::NewEnumerator1
    Uniform Sprite Size Max=14
    Uniform Sprite Size Min=10
    UsePositionOffset=false
    Write Material Random=true
   )
   &PartitionParticles(Number Of Partitions=$Emitter.PortCount)
   &SkeletalMeshLocation(
    Bone / Socket Index=$Particles.PartitionParticles.Partition
    Bone / Socket Sampling Mode=::NewEnumerator5
    MaskSpawnGroup=false
    Mesh Sampling Type=::NewEnumerator6
    Sampled Position Offset=&AddVectorToPosition(
     Position=&MaskVectorBySpawnGroup(Masked Vector=<-85 0 0>, Passthrough Vector=<-80 0 0>)
     Vector=$User.DTCarOffset
    )
    Skeletal Mesh=$System.SkeletalMesh
   )
   &AddVelocity(Velocity=&RandomRangeVector(Minimum=<-400 0 0>, Maximum=<-600 0 0>))
   &ShapeLocation(
    MaskSpawnGroup=false
    Override Seed=false
    Sphere Radius=&MaskFloatBySpawnGroup(Masked Float=25, Passthrough Float=0)
    UseOwnerScale=true
   )
  ]
  particleUpdateScript [
   &ParticleState()
   &SolveForcesAndVelocity(Clamp Velocity=false, Limit Acceleration=false)
   &ScaleColor(
    Scale Alpha=&MaskFloatBySpawnGroup(
     Masked Float=&FloatFromCurve(FloatCurve=DataInterface() {propertyValues {Curve {keys [<at 0 0>,<at 0.2 1>,<at 1 0>]}}})
     Passthrough Float=1.5
    )
    Scale RGB=&MaskVectorBySpawnGroup(
     Masked Vector=&MakeCustomVectorFromBool(
      Bool=$User.bIsPainted
      True Vector=&Multiply_VectorByFloat(
       Vector=&SelectVectorFromArray(
        Array Sampling Mode=::NewEnumerator0
        Direct Array Index=?
        Vector Selection Array=DataInterface() {propertyValues {FloatData [<0.2 0.2 0.2>, <4 4 4>]}}
        Random Seed=0
       )
       Float=$User.StarsSpritesScalar
      )
      False Vector=&VectorFromFloat(Value=&RandomRangeFloat(Minimum=3, Maximum=2))
     )
     Passthrough Vector=<1 1 1>
    )
    ScaleA=true
    ScaleRGB=true
   ) {module=ScaleColor002}
  ]
  rendererClasses [&NiagaraSpriteRendererProperties]
  renderers [{rendererClass=&NiagaraSpriteRendererProperties}]
 }
 mesh_glow {
  emitterUpdateScript [
   &EmitterState(
    Inactive Response=::NewEnumerator1
    Life Cycle Mode=::NewEnumerator1
    Loop Behavior=::NewEnumerator1
    Loop Duration Mode=::NewEnumerator1
    UseLoopDelay=false
   )
   &VC_FilteredSocketCount()
   &SpawnBurst_Instantaneous(Spawn Count=2, Use Spawn Probability=false, UseLoopCountLimit=false)
   &SpawnBurst_Instantaneous(Spawn Count=2, Spawn Group=1, Use Spawn Probability=false, UseLoopCountLimit=false) {module=SpawnBurst_Instantaneous001}
   &SpawnRate(SpawnRate=2.2, Use Spawn Probability=false) {enabled=false}
  ]
  particleSpawnScript [
   /SetVariables(Particles.VisTAgHeadStar:NiagaraInt32=&SpawnGroup()) {isSetParameters=true, module=SetVariables_#2}
   &VC_PaintedVariantsOverride() {enabled=false}
   &InitializeParticle(
    Color=&MakeCustomLinearColorFromBool(Bool=$User.bIsPainted, True Color=$User.PaintedColor, False Color=<rgba 1 0.15 1 0.5>)
    Lifetime=/VehicleCosmetics/FX/Niagara/Modules/MergeEmitters/VC_FloatBySpawnGroup(
     SpawnGroup or ExecIndex:NiagaraInt32=0
     CustomInt:NiagaraInt32=0
     MaximumSpawnGroups:NiagaraInt32=3
     Float SG 0:NiagaraFloat=1
     Float SG 1:NiagaraFloat=0.6
     Float SG 2:NiagaraFloat=0.1
     Float SG 3:NiagaraFloat=0
     Float SG 4:NiagaraFloat=0
     Float SG 5:NiagaraFloat=0
    )
    Mesh Scale Mode=::NewEnumerator3
    Mesh Uniform Scale=&MaskFloatBySpawnGroup(Masked Float=5, Passthrough Float=3.3)
    Sprite Rotation Mode=::NewEnumerator2
    UsePositionOffset=false
    Write Material Random=true
   )
   &PartitionParticles(Number Of Partitions=$Emitter.PortCount)
   &SkeletalMeshLocation(
    Bone / Socket Index=$Particles.PartitionParticles.Partition
    Bone / Socket Sampling Mode=::NewEnumerator5
    MaskSpawnGroup=false
    Mesh Sampling Type=::NewEnumerator6
    Sampled Position Offset=&AddVectorToPosition(Position=<-80 0 0>, Vector=$User.DTCarOffset)
    Skeletal Mesh=$System.SkeletalMesh
   )
   &AddVelocity(Velocity=&MaskVectorBySpawnGroup(Masked Vector=<-30 0 0>, Passthrough Vector=<0 0 0>, Spawn Group Mask=2))
   /VehicleCosmetics/FX/Niagara/Modules/MergeEmitters/VC_ExtraBindingColors(
    ExtraColors=2
    R1 Color:LinearColor=&MakeCustomLinearColorFromBool(Bool=$User.bIsPainted, True Color=$User.PaintedColor, False Color=<rgba 1 1 1 1>)
    R2 Color:LinearColor=&MakeCustomLinearColorFromBool(Bool=$User.bIsPainted, True Color=$User.PaintedColor, False Color=<rgba 1 0.4 1 0.8261>)
   )
  ]
  particleUpdateScript [
   &ParticleState(Kill Particles When Lifetime Has Elapsed=false, Loop Particles Lifetime=true)
   &ScaleMeshSize(Scale Factor=&MaskVectorBySpawnGroup(
     Masked Vector=<1 1 1>
     Passthrough Vector=&VectorFromFloat(Value=&FloatFromCurve(FloatCurve=DataInterface() {propertyValues {Curve {keys [<at 0 1>,<at 0.2 0.85>,<at 0.5 1>,<at 0.8 0.85>,<at 1 1>]}}}))
    ))
   &SolveForcesAndVelocity(Clamp Velocity=false, Limit Acceleration=false)
   &ScaleColor(
    Scale Alpha=&MakeCustomFloatFromBool(Bool=$User.bIsPainted, True Float=1.8, False Float=0.7)
    Scale RGB=&MaskVectorBySpawnGroup(
     Masked Vector=&VectorFromFloat(Value=&MakeCustomFloatFromBool(Bool=$User.bIsPainted, True Float=0.2, False Float=0.7))
     Passthrough Vector=&VectorFromFloat(Value=&MakeCustomFloatFromBool(Bool=$User.bIsPainted, True Float=0.3, False Float=0.7))
    )
    ScaleA=true
    ScaleRGB=true
   )
   /SetVariables(
    Particles.R1Color2:LinearColor=&VC_DynamicScaleColor(
     Color to Scale:LinearColor=$Particles.Initial.R1 Color
     RGB Scale:Vector3f=&VectorFromFloat(Value=&MakeCustomFloatFromBool(Bool=$User.bIsPainted, True Float=0.5, False Float=0.7))
     Alpha Scale:NiagaraFloat=1
    )
    Particles.R2Color2:LinearColor=&VC_DynamicScaleColor(
     Color to Scale:LinearColor=$Particles.Initial.R2 Color
     RGB Scale:Vector3f=&VectorFromFloat(Value=&MakeCustomFloatFromBool(Bool=$User.bIsPainted, True Float=&Multiply_Float(A=$User.StarOutlineScalar, B=1), False Float=1.75))
     Alpha Scale:NiagaraFloat=1
    )
    Particles.R3Color2:LinearColor=&VC_DynamicScaleColor(
     Color to Scale:LinearColor=$Particles.Initial.R2 Color
     RGB Scale:Vector3f=&VectorFromFloat(Value=&MakeCustomFloatFromBool(Bool=$User.bIsPainted, True Float=&Multiply_Float(A=$User.StarOutlineScalar, B=1.4), False Float=5))
     Alpha Scale:NiagaraFloat=&FloatFromCurve(FloatCurve=DataInterface() {propertyValues {Curve {keys [{arriveTangent=-1.4726, interpMode=RCIM_Cubic, leaveTangent=-1.4726, tangentMode=RCTM_User, time=0, value=1}, <at 1 0 RCIM_Cubic>]}}})
    )
    Particles.R3SizeStarOutline:Vector3f=&Multiply_Vector(
     A=&VectorFromFloat(Value=3.3)
     B=&VectorFromFloat(Value=&FloatFromCurve(
       FloatCurve=DataInterface() {propertyValues {Curve {keys [
           {arriveTangent=-0.0233, interpMode=RCIM_Cubic, leaveTangent=-0.0233, tangentMode=RCTM_User, time=0, value=0.6}
           {arriveTangentWeight=0.334, interpMode=RCIM_Cubic, tangentMode=RCTM_User, time=1, value=1}
          ]}}}
       Scale Curve=1.2
      ))
    )
   ) {isSetParameters=true, module=SetVariables_#3}
  ]
  rendererClasses [&NiagaraMeshRendererProperties, &NiagaraSpriteRendererProperties]
  renderers [
   {rendererClass=&NiagaraMeshRendererProperties, rendererIndex=0}
   {rendererClass=&NiagaraMeshRendererProperties, rendererIndex=1}
   {rendererClass=&NiagaraMeshRendererProperties, rendererIndex=2}
   {rendererClass=&NiagaraMeshRendererProperties, rendererIndex=3}
   {rendererClass=&NiagaraSpriteRendererProperties, rendererIndex=4}
  ]
 }
 mesh_glow001 {
  emitterSpawnScript [
   /SetVariables(Emitter.No Painted:NiagaraBool=&Boolean_NotOperation(Bool=$User.bIsPainted)) {isSetParameters=true}
   /SetVariables(Emitter.Painted:NiagaraBool=&Boolean_NotOperation(Bool=$StackContext.No Painted)) {isSetParameters=true}
  ]
  emitterUpdateScript [
   &EmitterState(
    Inactive Response=::NewEnumerator1
    Life Cycle Mode=::NewEnumerator1
    Loop Behavior=::NewEnumerator1
    Loop Duration Mode=::NewEnumerator1
    UseLoopDelay=false
   )
   &VC_FilteredSocketCount()
   &SpawnBurst_Instantaneous(Spawn Count=2, Use Spawn Probability=false, UseLoopCountLimit=false)
   &SpawnBurst_Instantaneous(Spawn Count=2, Spawn Group=1, Use Spawn Probability=false, UseLoopCountLimit=false) {module=SpawnBurst_Instantaneous001}
   &SpawnRate(SpawnRate=2.2, Use Spawn Probability=false) {enabled=false}
  ]
  particleSpawnScript [
   /SetVariables(Particles.VisTAgHeadStar:NiagaraInt32=&SpawnGroup()) {isSetParameters=true, module=SetVariables_#2}
   &VC_PaintedVariantsOverride(Pink=<rgba 0.1458 0.0175 0.0992 1>)
   &InitializeParticle(
    Color=&MakeCustomLinearColorFromBool(Bool=$User.bIsPainted, True Color=$User.PaintedColor, False Color=<rgba 1 0.15 1 1.2>)
    Lifetime=&VC_FloatBySpawnGroup(
     SpawnGroup or ExecIndex:NiagaraInt32=0
     CustomInt:NiagaraInt32=0
     MaximumSpawnGroups:NiagaraInt32=3
     Float SG 0:NiagaraFloat=1
     Float SG 1:NiagaraFloat=0.6
     Float SG 2:NiagaraFloat=0.1
     Float SG 3:NiagaraFloat=0
     Float SG 4:NiagaraFloat=0
     Float SG 5:NiagaraFloat=0
    )
    Mesh Scale Mode=::NewEnumerator3
    Mesh Uniform Scale=&MaskFloatBySpawnGroup(Masked Float=5, Passthrough Float=3.3)
    Sprite Rotation Mode=::NewEnumerator2
    UsePositionOffset=false
    Write Material Random=true
   )
   &PartitionParticles(Number Of Partitions=$Emitter.PortCount)
   &SkeletalMeshLocation(
    Bone / Socket Index=$Particles.PartitionParticles.Partition
    Bone / Socket Sampling Mode=::NewEnumerator5
    MaskSpawnGroup=false
    Mesh Sampling Type=::NewEnumerator6
    Sampled Position Offset=&AddVectorToPosition(Position=<-80 0 0>, Vector=$User.DTCarOffset)
    Skeletal Mesh=$System.SkeletalMesh
   )
   &AddVelocity(Velocity=&MaskVectorBySpawnGroup(Masked Vector=<-30 0 0>, Passthrough Vector=<0 0 0>, Spawn Group Mask=2))
   &VC_ExtraBindingColors(
    ExtraColors=2
    R1 Color:LinearColor=&MakeCustomLinearColorFromBool(Bool=$User.bIsPainted, True Color=$User.PaintedColor, False Color=<rgba 1 1 1 1>)
    R2 Color:LinearColor=&MakeCustomLinearColorFromBool(Bool=$User.bIsPainted, True Color=$User.PaintedColor, False Color=<rgba 1 0.4 1 1>)
   )
  ]
  particleUpdateScript [
   &ParticleState(Kill Particles When Lifetime Has Elapsed=false, Loop Particles Lifetime=true)
   &ScaleMeshSize(Scale Factor=&MaskVectorBySpawnGroup(
     Masked Vector=<1 1 1>
     Passthrough Vector=&VectorFromFloat(Value=&FloatFromCurve(FloatCurve=DataInterface() {propertyValues {Curve {keys [<at 0 1>,<at 0.2 0.85>,<at 0.5 1>,<at 0.8 0.85>,<at 1 1>]}}}))
    ))
   &SolveForcesAndVelocity(Clamp Velocity=false, Limit Acceleration=false)
   &DynamicMaterialParameters(
    Index 1 Param 1=1.5
    Index 2 Param 1=1
    Index 2 Param 2=&MakeCustomFloatFromBool(Bool=$User.bIsPainted, True Float=1, False Float=0)
    Module.Index1Param0Write=true
    Module.Index1Param1Write=true
    Module.Index1Param2Write=true
    Module.Index1Param3Write=true
    Module.Index2Param0Write=true
    Module.Index2Param1Write=true
    Module.Index2Param2Write=true
    Module.Index2Param3Write=true
    Module.Index3Param0Write=true
    Module.Index3Param1Write=true
    Module.Index3Param2Write=true
    Module.Index3Param3Write=true
    Write Parameter Index 0=false
    Write Parameter Index 1=true
    Write Parameter Index 2=true
    Write Parameter Index 3=true
   )
   &ScaleColor(
    Scale Alpha=&MakeCustomFloatFromBool(Bool=$User.bIsPainted, True Float=1.8, False Float=0.7)
    Scale RGB=&MaskVectorBySpawnGroup(
     Masked Vector=&VectorFromFloat(Value=&MakeCustomFloatFromBool(Bool=$User.bIsPainted, True Float=0.2, False Float=0.7))
     Passthrough Vector=&VectorFromFloat(Value=&MakeCustomFloatFromBool(Bool=$User.bIsPainted, True Float=0.3, False Float=0.7))
    )
    ScaleA=true
    ScaleRGB=true
   )
   /SetVariables(
    Particles.R1Color2:LinearColor=&VC_DynamicScaleColor(
     Color to Scale:LinearColor=$Particles.Initial.R1 Color
     RGB Scale:Vector3f=&VectorFromFloat(Value=&MakeCustomFloatFromBool(Bool=$User.bIsPainted, True Float=0.45, False Float=0.45))
     Alpha Scale:NiagaraFloat=1
    )
    Particles.R2Color2:LinearColor=&VC_DynamicScaleColor(
     Color to Scale:LinearColor=$Particles.Initial.R2 Color
     RGB Scale:Vector3f=&VectorFromFloat(Value=&MakeCustomFloatFromBool(Bool=$User.bIsPainted, True Float=&Multiply_Float(A=$User.StarOutlineScalar, B=1.3), False Float=5))
     Alpha Scale:NiagaraFloat=1
    )
    Particles.R3Color2:LinearColor=&VC_DynamicScaleColor(
     Color to Scale:LinearColor=$Particles.Initial.R2 Color
     RGB Scale:Vector3f=&VectorFromFloat(Value=&MakeCustomFloatFromBool(Bool=$User.bIsPainted, True Float=&Multiply_Float(A=$User.StarOutlineScalar, B=1.4), False Float=5))
     Alpha Scale:NiagaraFloat=&FloatFromCurve(FloatCurve=DataInterface() {propertyValues {Curve {keys [{arriveTangent=-1.4726, interpMode=RCIM_Cubic, leaveTangent=-1.4726, tangentMode=RCTM_User, time=0, value=1}, <at 1 0 RCIM_Cubic>]}}})
    )
    Particles.R3SizeStarOutline:Vector3f=&Multiply_Vector(
     A=&VectorFromFloat(Value=3.3)
     B=&VectorFromFloat(Value=&FloatFromCurve(
       FloatCurve=DataInterface() {propertyValues {Curve {keys [
           {arriveTangent=-0.0233, interpMode=RCIM_Cubic, leaveTangent=-0.0233, tangentMode=RCTM_User, time=0, value=0.6}
           {arriveTangentWeight=0.334, interpMode=RCIM_Cubic, tangentMode=RCTM_User, time=1, value=1}
          ]}}}
       Scale Curve=1.2
      ))
    )
   ) {isSetParameters=true, module=SetVariables_#3}
  ]
  rendererClasses [&NiagaraMeshRendererProperties, &NiagaraSpriteRendererProperties]
  renderers [
   {rendererClass=&NiagaraMeshRendererProperties, rendererIndex=0}
   {rendererClass=&NiagaraMeshRendererProperties, rendererIndex=1}
   {rendererClass=&NiagaraMeshRendererProperties, rendererIndex=2}
   {rendererClass=&NiagaraMeshRendererProperties, rendererIndex=3}
   {rendererClass=&NiagaraSpriteRendererProperties, rendererIndex=4}
  ]
 }
 mesh_staroutline001 {
  emitterUpdateScript [
   &EmitterState()
   &VC_FilteredSocketCount()
   &SpawnRate(SpawnRate=8, Use Spawn Probability=false)
  ]
  particleSpawnScript [
   &VC_PaintedVariantsOverride(
    Black=<rgba 0.025 0.025 0.025 1>
    BurntSienna=<rgba 0.5739 0.1214 0 1>
    Cobalt=<rgba 0.0343 0 2 1>
    Crimson=<rgba 0.7708 0 0.0209 1>
    DefaultColor=<rgba 1 0.651 1 0.2957>
    Lime=<rgba 0.4696 0.9391 0 1>
    Pink=<rgba 1 0.1198 0.6799 1>
    Purple=<rgba 0.8251 0 2 1>
    Saffron=<rgba 1 0.7667 0 1>
    SkyBlue=<rgba 0 0.6654 1 1>
   )
   &InitializeParticle(
    Color=$Particles.Color
    Lifetime=0.6
    Mesh Scale Mode=::NewEnumerator3
    Mesh Uniform Scale=&MakeCustomFloatFromBool(Bool=$User.bIsPainted, True Float=3.5, False Float=3.5)
    Sprite Rotation Mode=::NewEnumerator2
    UsePositionOffset=false
    Write Material Random=true
   )
   &PartitionParticles(Number Of Partitions=$Emitter.PortCount)
   &SkeletalMeshLocation(
    Bone / Socket Index=$Particles.PartitionParticles.Partition
    Bone / Socket Sampling Mode=::NewEnumerator5
    MaskSpawnGroup=false
    Mesh Sampling Type=::NewEnumerator6
    Sampled Position Offset=&AddVectorToPosition(Position=<-75 0 0>, Vector=$User.DTCarOffset)
    Skeletal Mesh=$System.SkeletalMesh
   )
   &AddVelocity(Velocity=<-30 0 0>)
  ]
  particleUpdateScript [
   &ParticleState()
   &ScaleMeshSize(Scale Factor=&VectorFromFloat(Value=&FloatFromCurve(
      FloatCurve=DataInterface() {propertyValues {Curve {keys [
          {arriveTangent=-0.0233, interpMode=RCIM_Cubic, leaveTangent=-0.0233, tangentMode=RCTM_User, time=0, value=0.6}
          {arriveTangentWeight=0.334, interpMode=RCIM_Cubic, tangentMode=RCTM_User, time=1, value=1}
         ]}}}
      Scale Curve=1.2
     )))
   &SolveForcesAndVelocity(Clamp Velocity=false, Limit Acceleration=false)
   &ScaleColor(
    Scale Alpha=&MakeCustomFloatFromBool(
     Bool=$User.bIsPainted
     True Float=&FloatFromCurve(FloatCurve=DataInterface() {propertyValues {Curve {keys [{arriveTangent=-1.4726, interpMode=RCIM_Cubic, leaveTangent=-1.4726, tangentMode=RCTM_User, time=0, value=1}, <at 1 0 RCIM_Cubic>]}}})
     False Float=&FloatFromCurve(FloatCurve=DataInterface() {propertyValues {Curve {keys [
          {arriveTangent=-1.6776, interpMode=RCIM_Cubic, leaveTangent=-1.6776, leaveTangentWeight=0.3385, tangentMode=RCTM_User, time=0, value=1}
          <at 1 0 RCIM_Cubic>
         ]}}})
    )
    Scale RGB=&VectorFromFloat(Value=&MakeCustomFloatFromBool(Bool=$User.bIsPainted, True Float=&Multiply_Float(A=$User.StarOutlineScalar, B=1.4), False Float=5))
    ScaleA=true
    ScaleRGB=true
   )
  ]
  rendererClasses [&NiagaraMeshRendererProperties, &NiagaraSpriteRendererProperties]
  renderers [
   {rendererClass=&NiagaraMeshRendererProperties, rendererIndex=0}
   {rendererClass=&NiagaraMeshRendererProperties, rendererIndex=1}
   {rendererClass=&NiagaraSpriteRendererProperties, rendererIndex=2}
  ]
 }
 ribbon001 {
  emitterSpawnScript [
   /SetVariables(Emitter.No Painted:NiagaraBool=&Boolean_NotOperation(Bool=$User.bIsPainted)) {isSetParameters=true}
   /SetVariables(Emitter.Painted:NiagaraBool=&Boolean_NotOperation(Bool=$StackContext.No Painted)) {isSetParameters=true}
  ]
  emitterUpdateScript [
   &EmitterState(Life Cycle Mode=::NewEnumerator1, Loop Duration=1, UseLoopDelay=false)
   &VC_FilteredSocketCount()
   &VC_RateQualityHandler()
   &VC_SpawnRibbon_v1(Number of Ribbons=$Emitter.PortCount, Spawn Time=-1, SpawnRate=60)
  ]
  particleSpawnScript [
   &VC_PaintedVariantsOverride(
    Black=<rgba 0.025 0.025 0.025 2>
    BurntSienna=<rgba 0.1458 0.0451 0.0063 2>
    Cobalt=<rgba 0.0964 0.0917 0.8 1>
    Crimson=<rgba 0.8021 0.016 0.016 1>
    DefaultColor=<rgba 0.0729 0.0729 0.0729 1>
    ForestGreen=<rgba 0.0904 0.5365 0.0168 1>
    Grey=<rgba 0.2062 0.2133 0.2344 1>
    Lime=<rgba 0.2713 0.6354 0 1>
    Orange=<rgba 0.76 0.1773 0 1>
    Pink=<rgba 0.804 0.4329 0.9896 1>
    Purple=<rgba 0.1935 0 0.8 1>
    R2 Black:LinearColor=<rgba 1 1 1 1>
    R2 BurntSienna:LinearColor=<rgba 0.75 0.5346 0 2>
    R2 Cobalt:LinearColor=<rgba 0.9 0.9388 3 1>
    R2 Crimson:LinearColor=<rgba 1 0 0.0271 2>
    R2 DefaultColor:LinearColor=<rgba 3 0.9 2.6688 1>
    R2 ForestGreen:LinearColor=<rgba 0.5078 1.5 0.5579 1.5>
    R2 Grey:LinearColor=<rgba 1 1 1 2>
    R2 Lime:LinearColor=<rgba 0.945 1 0.2969 2>
    R2 Orange:LinearColor=<rgba 2 0.6976 0 2>
    R2 Pink:LinearColor=<rgba 2.5 1.3542 2.0833 2>
    R2 Purple:LinearColor=<rgba 1.134 0.5312 2 2>
    R2 Saffron:LinearColor=<rgba 1.25 1.0647 0.4557 1.25>
    R2 SkyBlue:LinearColor=<rgba 0 3.3269 5 1>
    R2 TitaniumWhite:LinearColor=<rgba 2.5 2.5 2.5 1>
    Saffron=<rgba 0.5756 0.5833 0.0547 1>
    SkyBlue=<rgba 0.0291 0.3372 0.6979 1>
    TitaniumWhite=<rgba 0.8776 0.8776 0.901 0.4261>
    Use R1 Painted Overrides=true
    Use R2 Painted Overrides=true
   )
   &VC_InitializeRibbon_v1(
    Color=$Particles.Color
    Custom Offset=true
    Enable Skeletal Position=true
    Lifetime=1
    Local Vector Offset=&AddVectorToPosition(Position=<-55 0 0>, Vector=$User.DTCarOffset)
    Offset Amount=1
    Ribbon Width=100
    Write Material Random=false
   )
   &AddVelocity(Velocity=<-1300 0 0>) {module=AddVelocity001}
  ]
  particleUpdateScript [
   /SetVariables(Particles.DotProduct:NiagaraFloat=&DotProduct(
     Vector A=/Niagara/DynamicInputs/Math/TransformVectorByMatrix(
      Vector:Vector3f=<1 0 0>
      Transformation Matrix:NiagaraMatrix=$Engine.Owner.SystemLocalToWorld
      Transform Vector as Position:NiagaraBool=false
     )
     Vector B=&NormalizeVector(
      Vector To Normalize=&SubtractVectorFromPosition(
       Position=$Engine.Owner.Position
       Vector=&ReturnCameraPosition(Coordinate Space:ENiagaraCoordinateSpace=::Simulation, Camera Query:NiagaraDataInterfaceCamera=DataInterface())
      )
      Safe Normalize=true
      Safe Fallback Threshold=0
      Safe Fallback Vector=<0 0 0.001>
     )
    )) {isSetParameters=true}
   /SetVariables(
    Particles.DefaultColor:LinearColor=<rgba 0.7 0.7 0.7 0>
    Particles.OutlineColor:LinearColor=&MultiplyLinearColor(
     Linear Color=$Particles.R2 Color
     Scale Factor=&MakeVector4FromVectorAndFloat(
      "Vector (RGB)"=&VectorFromFloat(Value=&Lerp_Float(
        A=2
        B=1
        Alpha=&NormalizeFloat(Value=&Power(a=&Add_Float(A=$Particles.DotProduct, B=0.1), b=1.5), Min=0, Max=1)
       ))
      "Float (Alpha)"=&FloatFromCurve(FloatCurve=DataInterface() {propertyValues {Curve {
          defaultValue=1
          keys [{interpMode=RCIM_Cubic, tangentMode=RCTM_User, time=0, value=1}, {interpMode=RCIM_Cubic, tangentMode=RCTM_User, time=0.9, value=0}]
         }}})
     )
    )
   ) {isSetParameters=true}
   &ParticleState()
   &InheritSourceMovement(
    Applied Position Scale=&Multiply_VectorByFloat(
     Vector=<1 1 1>
     Float=&FloatFromCurve(FloatCurve=DataInterface() {propertyValues {Curve {keys [{interpMode=RCIM_Cubic, tangentMode=RCTM_Break, time=0, value=1}, {interpMode=RCIM_Cubic, tangentMode=RCTM_Break, time=1, value=0}]}}})
    )
    ApplyForce=false
    ApplyPosition=true
    ApplyPositionFalloff=false
   )
   &SolveForcesAndVelocity(Clamp Velocity=false, Limit Acceleration=false)
   &VC_UpdateRibbon_v1(Enable Skeletal Position=true)
   &ScaleRibbonWidth(Ribbon Width Scale=&FloatFromCurve(FloatCurve=DataInterface() {propertyValues {Curve {keys [
         {arriveTangent=0.3269, interpMode=RCIM_Cubic, leaveTangent=0.3269, tangentMode=RCTM_User, time=0, value=0.6}
         {arriveTangentWeight=1.0346, interpMode=RCIM_Cubic, tangentMode=RCTM_User, time=1, value=1.5}
        ]}}}))
   &ScaleColor(
    Scale Alpha=&FloatFromCurve(FloatCurve=DataInterface() {propertyValues {Curve {keys [<at 0.5 1 RCIM_Cubic>,<at 1 0 RCIM_Cubic>]}}}, Scale Curve=0.7769)
    Scale RGB=&Multiply_VectorByFloat(Vector=<1 1 1>, Float=$User.RibbonInsideColor)
    ScaleA=true
    ScaleRGB=true
   ) {enabled=false}
  ]
  rendererClasses [&NiagaraRibbonRendererProperties]
  renderers [
   {rendererClass=&NiagaraRibbonRendererProperties}
   {rendererClass=&NiagaraRibbonRendererProperties, rendererIndex=1}
   {rendererClass=&NiagaraRibbonRendererProperties, rendererIndex=2}
   {rendererClass=&NiagaraRibbonRendererProperties, rendererIndex=3}
  ]
 }
}
summary {
 systemName=NS_Boost_PopStar
 emitters [
  {emitterName=ribbon001, rendererClasses [&NiagaraRibbonRendererProperties]}
  {emitterName=mesh_glow, rendererClasses [&NiagaraMeshRendererProperties, &NiagaraSpriteRendererProperties]}
  {emitterName=StarSprites001, rendererClasses [&NiagaraSpriteRendererProperties]}
  {
   emitterName=mesh_staroutline001
   rendererClasses [&NiagaraMeshRendererProperties, &NiagaraSpriteRendererProperties]
  }
  {emitterName=mesh_glow001, rendererClasses [&NiagaraMeshRendererProperties, &NiagaraSpriteRendererProperties]}
 ]
 userVariables [
  User.bIsFrontendPreview {type=NiagaraBool, defaultValue=true}
  User.bIsPainted {type=NiagaraBool}
  User.ColorVariantID {type=NiagaraInt32}
  User.DTCarOffset {type=Vector3f, description="", defaultValue {struct=Vector3f, value=<0 0 0>}}
  User.PaintedColor {type=LinearColor, description="", defaultValue {struct=LinearColor, value=<rgba 1 1 1 1>}}
  User.PaintedScalar {type=NiagaraFloat, defaultValue=1, description=""}
  User.PortCount {type=NiagaraInt32, defaultValue=2}
  User.RibbonHLScalar {type=NiagaraFloat, defaultValue=4, description=""}
  User.RibbonInsideColor {type=NiagaraFloat, defaultValue=1, description=""}
  User.RibbonPaintedScalar {type=NiagaraFloat, defaultValue=1, description=""}
  User.SingleBoostPos {type=Vector3f, description="", defaultValue {struct=Vector3f, value=<0 0 0>}}
  User.StarOutlineScalar {type=NiagaraFloat, defaultValue=1, description=""}
  User.StarsSpritesScalar {type=NiagaraFloat, defaultValue=1, description=""}
 ]
}
